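; A conversation requirement that gates content on a curve-table-driven flag.
; IsRequirementSatisfied reads the CT Check ScalableFloat as a boolean and returns either Passed or FailedAndHidden, with an Invert variable flipping the outcome.
; A failed check always hides the gated content, and the conversation context is not used.
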(fn IsRequirementSatisfied
  (params (Context :type Conversation%20Context%20Structure%20%28by%20ref%29))
  (return (select (ScalableFloat|GetValueasBool (Variables|Default|GetCT%20Check)) (select (Variables|Default|GetInvert) "FailedAndHidden" "Passed") (select (Variables|Default|GetInvert) "Passed" "FailedAndHidden"))))
