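; Conversation requirement (Requirement_Conversation_Equipped_Item) that gates dialogue on a required item gameplay tag.
; It resolves the participant's player controller from the conversation context and checks its Fort ability system component for that tag.
; With invertedCondition set, a present tag returns failedRequirementBehavior and an absent one passes.
; The non-inverted branch makes the same check, but its outcomes are empty here, with "Passed" as the final return.
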(fn IsRequirementSatisfied
  (params (Context :type Conversation%20Context%20Structure%20%28by%20ref%29))
  (bind _requireditemtag (Variables|Default|GetrequiredItemTag))
  (bind _returnvalue (GameplayTags|MakeGameplayTagContainerfromTag _requireditemtag))
  (bind _returnvalue_2 (Class|FortConversationContextLibrary|GetParticipantPlayerController Context))
  (bind _output_get (LocalVariables|Set_invertedCondition (Variables|Default|GetinvertedCondition)))
  (if _output_get
    (bind _output_get_2 (LocalVariables|Set_foundTag (not (GameplayTags|HasAnyMatchingGameplayTags (Ability|GetFortAbilitySystemComponent _returnvalue_2) _returnvalue))))
    (if (not _output_get_2)
      (return (Variables|Default|GetfailedRequirementBehavior)))
    (else
      (bind _output_get_3 (LocalVariables|Set_foundTag (GameplayTags|HasAnyMatchingGameplayTags (Ability|GetFortAbilitySystemComponent _returnvalue_2) _returnvalue)))
      (if _output_get_3
        _
        (else
          _))))
  (return "Passed"))
