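; A conversation requirement that shows a dialogue option only when the player participant carries a configured gameplay tag.
; It gets the player actor from the conversation context and tests it for the Tag to Check variable through the gameplay tag interface.
; A match returns Passed; otherwise it returns FailedAndHidden, so the option is hidden rather than shown as unavailable.
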(fn IsRequirementSatisfied
  (params (Context :type Conversation%20Context%20Structure%20%28by%20ref%29))
  (bind _hastag (CallFunction|PlayerHasTag Context))
  (if _hastag
    (return "Passed")
    (else
      (return "FailedAndHidden"))))

(fn PlayerHasTag
  (params (Context :type Conversation%20Context%20Structure))
  (bind _returnvalue (Class|FortConversationContextLibrary|GetParticipantActorPlayer Context))
  (bind _returnvalue_2 (GameplayTags|HasMatchingGameplayTag(Message) _returnvalue (Variables|Default|GetTag%20to%20Check)))
  (return _returnvalue_2))
