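; Conversation requirement that gates a dialogue option on the NPC participant's conversation entry tag.
; IsRequirementSatisfied gets the NPC participant component from the conversation context, validates it and casts it to the NPC conversation component.
; HasConversationTag compares that component's entry tag to the requirement's configured tag, with optional exact matching.
; A match returns Passed, while FailedAndHidden is the failure result.
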(fn HasConversationTag
  (params (NPCConversationComponent :type Fort%20Non%20Player%20Conversation%20Participant%20Component%20Object%20Reference))
  (return (GameplayTags|MatchesTag (Variables|/Script/FortniteGame.FortNonPlayerConversationParticipantComponent|GetConversationEntryTag NPCConversationComponent) (Variables|Default|GetConversationEntryTag) (Variables|Default|GetrequireExactMatch))))

(fn IsRequirementSatisfied
  (params (Context :type Conversation%20Context%20Structure%20%28by%20ref%29))
  (bind _returnvalue (Class|FortConversationContextLibrary|GetParticipantComponentNPC Context))
  (Utilities|IsValid _returnvalue
    (:Is%20Valid
      (bind _asnpcconversation_component (Utilities|Casting|CastToNPCConversationComponent _returnvalue)
        (:then
          (if (CallFunction|HasConversationTag _asnpcconversation_component)
            (return "Passed")
            (else
              _)))
        (:CastFailed
          _))
      (return "FailedAndHidden"))
    (:Is%20Not%20Valid
      _)))
